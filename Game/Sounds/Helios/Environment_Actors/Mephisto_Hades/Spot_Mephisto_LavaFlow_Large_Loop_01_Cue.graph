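# Looping spot ambience cue for a large lava flow in the Mephisto/Hades environment, played under the Athena ambient sound class.
# A single looping wave player feeds the root directly, with boosted volume and no randomization or mixing.
# Inline attenuation uses natural-sound falloff beyond an inner radius, with stereo spread and focus settings that favour in-focus sources over off-focus ones.
# Concurrency is also overridden on the cue itself.
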SoundCue /Game/Sounds/Helios/Environment_Actors/Mephisto_Hades/Spot_Mephisto_LavaFlow_Large_Loop_01_Cue
root=WavePlayer_1
nodes [{
  id=WavePlayer_1
  properties {bLooping=true, soundWaveAssetPtr=&/Spot_Mephisto_LavaFlow_Large_Loop}
 }]
properties {
 bOverrideAttenuation=true
 bOverrideConcurrency=true
 maxDistance=3200
 soundClassObject=&Fort_Ambient_Athena
 volumeMultiplier=3.759
 attenuationOverrides {
  distanceAlgorithm=NaturalSound
  falloffDistance=2500
  focusDistanceScale=0.9
  focusVolumeAttenuation=1.2
  nonFocusVolumeAttenuation=0.4
  stereoSpread=700
  attenuationShapeExtents {x=700}
  customAttenuationCurve {editorCurveData {keys [{arriveTangent=-2.3812, interpMode=RCIM_Cubic, leaveTangent=-2.3812, tangentMode=RCTM_User, time=0, value=1}, <at 1 0.01 RCIM_Cubic>]}}
 }
}
