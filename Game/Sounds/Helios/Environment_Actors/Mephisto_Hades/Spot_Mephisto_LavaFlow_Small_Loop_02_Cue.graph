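# Sound cue for a small looping lava-flow spot ambience in the Mephisto Hades environment set, routed to the Athena ambient sound class.
# A single looping Wave Player feeds the output directly, with a slight volume boost.
# Inline attenuation overrides the referenced fireplace attenuation asset, using natural-sound falloff, distance-driven low-pass filtering, stereo spread and focus scaling.
# Concurrency is also set to override.
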SoundCue /Game/Sounds/Helios/Environment_Actors/Mephisto_Hades/Spot_Mephisto_LavaFlow_Small_Loop_02_Cue
root=WavePlayer_0
nodes [{
  id=WavePlayer_0
  properties {bLooping=true, soundWaveAssetPtr=&/Spot_Mephisto_LavaFlow_Small_Loop_02}
 }]
properties {
 attenuationSettings=/Game/Sounds/Attenuation/Trans_VFX_Fireplace
 bOverrideAttenuation=true
 bOverrideConcurrency=true
 maxDistance=1000
 soundClassObject=&Fort_Ambient_Athena
 volumeMultiplier=1.253
 attenuationOverrides {
  bAttenuateWithLPF=true
  distanceAlgorithm=NaturalSound
  falloffDistance=600
  focusDistanceScale=0.9
  focusVolumeAttenuation=1.2
  lPFFrequencyAtMax=12000
  lPFFrequencyAtMin=500
  lPFRadiusMax=1000
  lPFRadiusMin=400
  nonFocusVolumeAttenuation=0.4
  stereoSpread=700
  customAttenuationCurve {editorCurveData {keys [{arriveTangent=-2.3812, interpMode=RCIM_Cubic, leaveTangent=-2.3812, tangentMode=RCTM_User, time=0, value=1}, <at 1 0.01 RCIM_Cubic>]}}
 }
}
